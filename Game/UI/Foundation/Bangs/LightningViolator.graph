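# LightningViolator is a UI callout badge with skewed text, background and drop shadow, plus an optional pointer tail.
# PreConstruct applies default text and color through custom events, tinting Image_BG and Image_ShadowTail, the visible tail despite its name.
# TailPosition, run at design time and on Construct, shows or collapses the tail, aligns it horizontally and mirrors it vertically via Y scale.
# The flip reads the ShowOnTop variable directly, ignoring the OnTop input.
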WidgetBlueprint /Game/UI/Foundation/Bangs/LightningViolator
graphs [
 EventGraph {nodes [
   n1 = EventColorBG!() -> (OutputDelegate, then, NewColor:Linear Color Structure) {is=title K2Node_CustomEvent}
   > SetColorAndOpacity(
    self:Image Object Reference=[Get Image_ShadowTail() -> (Image_ShadowTail:Image Object Reference), Get Image_BG() -> (Image_BG:Image Object Reference)]
    InColorAndOpacity:Linear Color Structure=@n1.NewColor
   )
   n2 = EventSetText!() -> (OutputDelegate, then, NewText:Text) {is=title K2Node_CustomEvent}
   > "SetText (Text)"(self=Get CommonTextBlock_Content() -> (CommonTextBlock_Content:Common Text Object Reference), InText=@n2.NewText)
   n3 = Event PreConstruct()
   > EventSetText!(execute, self, NewText:Text=Get DefaultViolatorText() -> (DefaultViolatorText:Text)) -> (then) {is=title K2Node_CallFunction}
   > EventColorBG!(execute, self, NewColor:Linear Color Structure=Get DefaultBGColor() -> (DefaultBGColor:Linear Color Structure)) -> (then) {is=title K2Node_CallFunction}
   > Branch(Condition=@n3.IsDesignTime)
   > TailPosition!(
    execute
    self
    ShowTail:bool=Get bShowTail() -> (bShowTail:bool)
    HPosition:EHorizontalAlignment Enum=Get TailHorizontalAlignment() -> (TailHorizontalAlignment:EHorizontalAlignment Enum=HAlign_Fill)
    OnTop:bool=Get ShowOnTop() -> (ShowOnTop:bool)
   ) -> (then)
   Event Construct()
   > TailPosition(ShowTail=Get bShowTail(), HPosition=Get TailHorizontalAlignment() -> (TailHorizontalAlignment=HAlign_Fill), OnTop=Get ShowOnTop())
  ]}
 TailPosition {nodes [
   n1 = TailPosition!() -> (then, ShowTail:bool=~, HPosition:EHorizontalAlignment Enum=HAlign_Fill, OnTop:bool=~) {is=title K2Node_FunctionEntry}
   > n2 = Branch(Condition=@n1)
   n3 = Get Overlay_Pointer() -> (Overlay_Pointer:Overlay Object Reference)
   > n4 = SetVisibility(self=@n3, InVisibility=SelfHitTestInvisible)
   SetHorizontalAlignment(
    execute=[@n6.then, @n4.then]
    self=SlotAsOverlaySlot(Widget=Get Overlay_Pointer())
    InHorizontalAlignment=Get HPosition() -> (HPosition:EHorizontalAlignment Enum=HAlign_Fill)
   )
   n5 = Get ShowOnTop()
   > SetRenderTransform!(
    execute
    self:Widget Object Reference=Get Overlay_Pointer()
    InTransform:Widget Transform Structure=Make Widget Transform!(
     Translation:Vector 2D Structure=(X=0,Y=0)
     Scale:Vector 2D Structure=Select(Option 0:Vector 2D Structure=(X=1,Y=1), Option 1:Vector 2D Structure=(X=1,Y=-1), Index=@n5) -> (ReturnValue:Vector 2D Structure)
     Shear:Vector 2D Structure=(X=0,Y=0)
     Angle:float=0
    ) -> (WidgetTransform:Widget Transform Structure) {is=title K2Node_MakeStruct}
   ) -> (then)
   n6 = SetVisibility(execute=@n2.else, self=@n3, InVisibility=Collapsed)
  ]}
]
variables [
 bShowTail {type=boolean, default=true}
 DefaultBGColor {type=LinearColor, default=<rgba 1 0 0 1>}
 DefaultViolatorText {type=string, default=Violator Text}
 ShowOnTop {type=boolean}
 TailHorizontalAlignment {type=EHorizontalAlignment, default=HAlign_Left}
]
widgets {
 info {inheritedWidgetCount=0, parentClass=&UserWidget, rootWidgetClass=&Overlay, widgetCount=13}
 widgets [
  Overlay_1 {widgetClassPath=&Overlay, widget=Overlay_1}
  Overlay_Body {
   widgetClassPath=&Overlay
   bIsVariable=true
   parent=Overlay_1
   widget=Overlay_Body
   slot {class=OverlaySlot, properties {horizontalAlignment=HAlign_Fill, verticalAlignment=VAlign_Fill}}
  }
  CommonBorderShadow {
   widgetClassPath=&CommonBorder
   parent=Overlay_Body
   widget=CommonBorderShadow
   properties {
    contentColorAndOpacity=<rgba 0.0052 0.0052 0.0056 0.4>
    padding=<ltrb 0 0 0 0>
    renderTransform {angle=2, scale=<1 1>, shear=<-2 0>, translation=<0 5>}
    renderTransformPivot=<0.1813 1.931>
   }
   slot {class=OverlaySlot, properties {horizontalAlignment=HAlign_Fill, verticalAlignment=VAlign_Fill, padding=<ltrb 0 -5 0 0>}}
  }
  OverlayShadow {
   widgetClassPath=&Overlay
   parent=CommonBorderShadow
   widget=OverlayShadow
   slot {class=BorderSlot, properties {padding=<ltrb 0 0 0 0>}}
  }
  Image_ShadowBG {
   widgetClassPath=&Image
   bIsVariable=true
   parent=OverlayShadow
   widget=Image_ShadowBG
   properties {
    brush {
     bIsDynamicallyLoaded=false
     drawAs=Box
     imageType=NoImage
     mirroring=NoMirror
     resourceName=None
     resourceObject=/Game/UI/Frontend/Showdown/Materials/M_UI_SkewBox_LiveBG
     tiling=NoTile
     imageSize=<32 32>
     margin=<ltrb 0.1 0.1 0.1 0.1>
     outlineSettings {
      bUseBrushTransparency=false
      roundingType=HalfHeightRadius
      width=0
      color {colorUseRule=UseColor_Specified, specifiedColor=<rgba 0 0 0 0>}
      cornerRadii=<0 0 0 0>
     }
     tintColor {colorUseRule=UseColor_Specified, specifiedColor=<rgba 1 1 1 1>}
     uVRegion {bIsValid=false, max=<0 0>, min=<0 0>}
    }
    colorAndOpacity=<rgba 1 0 0 1>
    renderTransform {angle=0, scale=<1 1>, shear=<-14 0>, translation=<0 0>}
   }
   slot {class=OverlaySlot, properties {horizontalAlignment=HAlign_Fill, verticalAlignment=VAlign_Fill}}
  }
  Image_BG {
   widgetClassPath=&Image
   bIsVariable=true
   parent=Overlay_Body
   widget=Image_BG
   properties {
    brush {
     bIsDynamicallyLoaded=false
     drawAs=Box
     imageType=NoImage
     mirroring=NoMirror
     resourceName=None
     resourceObject=&M_UI_SkewBox_LiveBG
     tiling=NoTile
     imageSize=<32 32>
     margin=<ltrb 0.1 0.1 0.1 0.1>
     outlineSettings {
      bUseBrushTransparency=false
      roundingType=HalfHeightRadius
      width=0
      color {colorUseRule=UseColor_Specified, specifiedColor=<rgba 0 0 0 0>}
      cornerRadii=<0 0 0 0>
     }
     tintColor {colorUseRule=UseColor_Specified, specifiedColor=<rgba 1 1 1 1>}
     uVRegion {bIsValid=false, max=<0 0>, min=<0 0>}
    }
    colorAndOpacity=<rgba 1 0 0 1>
    renderTransform {angle=0, scale=<1 1>, shear=<-14 0>, translation=<0 0>}
   }
   slot {class=OverlaySlot, properties {horizontalAlignment=HAlign_Fill, verticalAlignment=VAlign_Fill}}
  }
  Overlay_Pointer {
   widgetClassPath=&Overlay
   bIsVariable=true
   parent=Overlay_1
   widget=Overlay_Pointer
   slot {class=OverlaySlot, properties {verticalAlignment=VAlign_Fill}}
  }
  CommonBorderShadow_1 {
   widgetClassPath=&CommonBorder
   parent=Overlay_Pointer
   widget=CommonBorderShadow_1
   properties {
    horizontalAlignment=HAlign_Left
    verticalAlignment=VAlign_Center
    contentColorAndOpacity=<rgba 0.0052 0.0052 0.0056 0.4>
    padding=<ltrb 0 0 0 0>
    renderTransform {angle=2, scale=<1 1>, shear=<-2 0>, translation=<0 5>}
    renderTransformPivot=<0.1813 1.931>
   }
   slot {class=OverlaySlot, properties {horizontalAlignment=HAlign_Right, verticalAlignment=VAlign_Fill, padding=<ltrb 0 -5 0 0>}}
  }
  Image_Tail {
   widgetClassPath=&Image
   bIsVariable=true
   parent=CommonBorderShadow_1
   widget=Image_Tail
   properties {
    brush {
     bIsDynamicallyLoaded=false
     drawAs=Image
     imageType=NoImage
     mirroring=NoMirror
     resourceName=None
     resourceObject=/Game/UI/Frontend/Showdown/Materials/M_UI_LiveViolatorTail_Inst
     tiling=NoTile
     imageSize=<42 42>
     margin=<ltrb 0 0 0 0>
     outlineSettings {
      bUseBrushTransparency=false
      roundingType=HalfHeightRadius
      width=0
      color {colorUseRule=UseColor_Specified, specifiedColor=<rgba 0 0 0 0>}
      cornerRadii=<0 0 0 0>
     }
     tintColor {colorUseRule=UseColor_Specified, specifiedColor=<rgba 1 1 1 1>}
     uVRegion {bIsValid=false, max=<0 0>, min=<0 0>}
    }
    colorAndOpacity=<rgba 1 0 0 1>
    renderTransform {angle=0, scale=<1 1>, shear=<0 0>, translation=<-10 45>}
   }
   slot {class=BorderSlot, properties {horizontalAlignment=HAlign_Left, verticalAlignment=VAlign_Center, padding=<ltrb 0 0 0 0>}}
  }
  Image_ShadowTail {
   widgetClassPath=&Image
   bIsVariable=true
   parent=Overlay_Pointer
   widget=Image_ShadowTail
   properties {
    brush {
     bIsDynamicallyLoaded=false
     drawAs=Image
     imageType=NoImage
     mirroring=NoMirror
     resourceName=None
     resourceObject=&M_UI_LiveViolatorTail_Inst
     tiling=NoTile
     imageSize=<42 42>
     margin=<ltrb 0 0 0 0>
     outlineSettings {
      bUseBrushTransparency=false
      roundingType=HalfHeightRadius
      width=0
      color {colorUseRule=UseColor_Specified, specifiedColor=<rgba 0 0 0 0>}
      cornerRadii=<0 0 0 0>
     }
     tintColor {colorUseRule=UseColor_Specified, specifiedColor=<rgba 1 1 1 1>}
     uVRegion {bIsValid=false, max=<0 0>, min=<0 0>}
    }
    colorAndOpacity=<rgba 1 0 0 1>
    renderTransform {angle=0, scale=<1 1>, shear=<0 0>, translation=<-10 45>}
   }
   slot {class=OverlaySlot, properties {horizontalAlignment=HAlign_Right, verticalAlignment=VAlign_Center}}
  }
  SizeBox_0 {
   widgetClassPath=&SizeBox
   parent=Overlay_1
   widget=SizeBox_0
   properties {bOverride_MinDesiredWidth=true, minDesiredWidth=57.2752}
   slot {class=OverlaySlot, properties {padding=<ltrb 15 0 23 0>}}
  }
  HorizontalBox_0 {widgetClassPath=&HorizontalBox, parent=SizeBox_0, widget=HorizontalBox_0, slot {class=SizeBoxSlot}}
  CommonTextBlock_Content {
   widgetClassPath=&CommonTextBlock
   bIsVariable=true
   parent=HorizontalBox_0
   widget=CommonTextBlock_Content
   properties {
    text=Violator Text
    font {
     bForceMonospaced=false
     bMaterialIsStencil=false
     fontMaterial=null
     fontObject=&Burbank
     letterSpacing=0
     monospacedWidth=1
     size=35
     skewAmount=0
     typefaceFontName=Burbank-Black
     outlineSettings {
      bApplyOutlineToDropShadows=false
      bMiteredCorners=false
      bSeparateFillAlpha=false
      outlineMaterial=null
      outlineSize=0
      outlineColor=<rgba 0 0 0 1>
     }
    }
    renderTransform {angle=0, scale=<1 1>, shear=<-10 0>, translation=<0 0>}
   }
   slot {class=HorizontalBoxSlot, properties {padding=<ltrb 10 0 0 0>}}
  }
 ]
}
