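# Widget wrapper that overlays a notification "bang" badge on slotted content and shows a looping glow border during tutorial callouts.
# Bang state changes set visibility and count text, then pick a WidgetSwitcher page: big or small with optional count, super bang, or padlock.
# A count of exactly 5 selects the padlock, overriding the super and small flags.
# Functions play or reset intro animations, offset the badges, restyle the glow, or suppress the bang.
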WidgetBlueprint /Game/UI/Foundation/Bangs/NormalBangWrapper
animations [
 TutorialGlow {
  binding_count=1
  display_label=""
  length_seconds=1
  object=&self:TutorialGlow
  sequence {bindings [TutorialBorder {
     id=#1
     possessed_class=CommonBorder
     tracks [
      {
       type=MovieScene2DTransformTrack
       sections [{
         type=MovieScene2DTransformSection
         end_frame=20
         start_frame=0
         channels [
          Translation.X {type=MovieSceneScriptingFloatChannel}
          Translation.Y {type=MovieSceneScriptingFloatChannel}
          Angle {type=MovieSceneScriptingFloatChannel}
          Scale.X {type=MovieSceneScriptingFloatChannel, keys [[0,1],[20,2]]}
          Scale.Y {type=MovieSceneScriptingFloatChannel}
          Shear.X {type=MovieSceneScriptingFloatChannel}
          Shear.Y {type=MovieSceneScriptingFloatChannel}
         ]
        }]
      }
      {
       type=MovieSceneColorTrack
       display_name=Brush Color
       sections [{
         type=MovieSceneColorSection
         end_frame=20
         start_frame=0
         channels [
          Color.R {type=MovieSceneScriptingFloatChannel, const [1,1]}
          Color.G {type=MovieSceneScriptingFloatChannel, const [1,1]}
          Color.B {type=MovieSceneScriptingFloatChannel, const [1,1]}
          Color.A {type=MovieSceneScriptingFloatChannel, keys [[0,0],[17.9997,0.7562]]}
         ]
        }]
      }
     ]
    }]}
 }
 Intro {
  binding_count=1
  display_label=""
  length_seconds=0.2
  object=&self:Intro
  sequence {
   bindings [Overlay_1 {
     id=#2
     possessed_class=Overlay
     tracks [{
       type=MovieSceneFloatTrack
       display_name=Render Opacity
       sections [{type=MovieSceneFloatSection, end_frame=20, start_frame=0, channels [None {type=MovieSceneScriptingFloatChannel, keys [[0,0],[19.9983,1]]}]}]
      }]
    }]
   display_rate {numerator=100}
  }
 }
 IntroSuperBang {
  binding_count=1
  display_label=""
  length_seconds=1
  object=&self:IntroSuperBang
  sequence {bindings [SuperBang {
     id=#3
     possessed_class=SizeBox
     tracks [
      {
       type=MovieScene2DTransformTrack
       sections [{
         type=MovieScene2DTransformSection
         end_frame=20
         start_frame=0
         channels [
          Translation.X {type=MovieSceneScriptingFloatChannel, keys [[0,-32],[10,-30],[20,0]]}
          Translation.Y {type=MovieSceneScriptingFloatChannel, keys [[0,96],[10,0],[15,24],[20,0]]}
          Angle {type=MovieSceneScriptingFloatChannel}
          Scale.X {type=MovieSceneScriptingFloatChannel, keys [[0,0.75],[10,1],[15,2],[20,1]]}
          Scale.Y {type=MovieSceneScriptingFloatChannel, keys [[0,1.25],[10,1],[15,1.75],[20,1]]}
          Shear.X {type=MovieSceneScriptingFloatChannel}
          Shear.Y {type=MovieSceneScriptingFloatChannel}
         ]
        }]
      }
      {
       type=MovieSceneFloatTrack
       display_name=Render Opacity
       sections [{type=MovieSceneFloatSection, end_frame=20, start_frame=0, channels [None {type=MovieSceneScriptingFloatChannel, keys [[0,0],[10,1],[20,1]]}]}]
      }
     ]
    }]}
 }
]
graphs [
 Adjust Bang Image Translation {nodes [
   n1 = Adjust Bang Image Translation!() -> (then, Translation:Vector 2D Structure) {is=title K2Node_FunctionEntry}
   > SetRenderTranslation(
    self=[
     Get SuperBang() -> (SuperBang:Size Box Object Reference)
     Get LockBang() -> (LockBang:Size Box Object Reference)
     Get BigBang() -> (BigBang:Size Box Object Reference)
     Get BigBangCount() -> (BigBangCount:Size Box Object Reference)
     Get SmallBang() -> (SmallBang:Size Box Object Reference)
     Get SmallBangCount() -> (SmallBangCount:Size Box Object Reference)
    ]
    Translation=@n1
   )
  ]}
 EventGraph {nodes [
   n1 = SetVisibility(self=Get ModeSwitcher() -> (ModeSwitcher:Widget Switcher Object Reference), InVisibility=HitTestInvisible)
   Update Bang Icon Version!(execute=[@n4.then, @n1.then], self, InCount:int=Get Count() -> (Count:int=0)) -> (then)
   n2 = "To Text (Integer)"(Value=Get Count() -> (Count=0))
   > "SetText (Text)"(self=Get CountText() -> (CountText:Common Text Object Reference), InText=@n2)
   > "SetText (Text)"(self=Get CommonTextBlock_0() -> (CommonTextBlock_0:Common Text Object Reference), InText=@n2)
   n3 = Event OnBangStateChanged!() -> (OutputDelegate, then, bEnabled:bool, Count:int=0) {is=title K2Node_Event}
   > Set Count(Count:int=@n3.Count) -> (Output_Get:int=0)
   > n4 = SetVisibility(self=Get ModeSwitcher(), InVisibility=Select(Option 0=Collapsed, Option 1=HitTestInvisible, Index=@n3.bEnabled))
   Event OnStopCallout() {is=title K2Node_Event}
   > SetVisibility(self=Get TutorialBorder() -> (TutorialBorder:Common Border Object Reference), InVisibility=Hidden)
   > StopAnimation(InAnimation=Get TutorialGlow() -> (TutorialGlow:Widget Animation Object Reference))
   Event OnStartCallout!() -> (OutputDelegate, then, TutorialObjectiveName:Name=None, TutorialType:ETutorialType Enum=None) {is=title K2Node_Event}
   > SetVisibility(self=Get TutorialBorder(), InVisibility=Visible)
   > PlayAnimation(InAnimation=Get TutorialGlow(), NumLoopsToPlay=0, PlayMode=PingPong)
   Event PreConstruct()
   > Update Bang Icon Version(InCount=Get Count() -> (Count=0))
  ]}
 InitialReset {nodes [
   InitialReset() {is=title K2Node_FunctionEntry}
   n1 = Get Intro()
   > PlayAnimation(InAnimation=@n1, StartAtTime=GetEndTime(self=@n1), PlayMode=Reverse)
   n2 = Get IntroSuperBang() -> (IntroSuperBang:Widget Animation Object Reference)
   > PlayAnimation(InAnimation=@n2, StartAtTime=GetEndTime(self=@n2), PlayMode=Reverse, PlaybackSpeed=2)
  ]}
 SetShouldSuppressBang {nodes [
   n1 = SetShouldSuppressBang!() -> (then, ShouldSuppressBang:bool=~) {is=title K2Node_FunctionEntry}
   > n2 = Set ShouldSuppressBang(ShouldSuppressBang:bool=@n1) -> (Output_Get:bool)
   > SetVisibility(self=Get BangBox() -> (BangBox:Size Box Object Reference), InVisibility=Select(Option 0=HitTestInvisible, Option 1=Collapsed, Index=@n2))
  ]}
 SetTutorialBorderStyle {nodes [
   n1 = SetTutorialBorderStyle!() -> (then, BorderStyle:Common Border Style Class Reference) {is=title K2Node_FunctionEntry}
   > SetStyle!(execute, self:Common Border Object Reference=Get TutorialBorder(), InStyle:Common Border Style Class Reference=@n1) -> (then)
  ]}
 TriggerIntro {nodes [
   TriggerIntro() {is=title K2Node_FunctionEntry}
   > PlayAnimation(InAnimation=Get Intro())
   > PlayAnimation(InAnimation=Get IntroSuperBang(), PlaybackSpeed=2)
  ]}
 TriggerReset {nodes [
   TriggerReset() {is=title K2Node_FunctionEntry}
   > StopAnimation(InAnimation=Get Intro())
   > StopAnimation(InAnimation=Get IntroSuperBang())
   > InitialReset() {is=title K2Node_CallFunction}
  ]}
 Update Bang Icon Version {nodes [
   n1 = Update Bang Icon Version!() -> (then, InCount:int) {is=title K2Node_FunctionEntry}
   > n2 = Branch(Condition="Equal (Integer)"(A=@n1, B=5))
   > SetActiveWidgetIndex(self=Get ModeSwitcher(), Index=5)
   n3 = Branch(execute=@n2.else, Condition=Get bUseSuperBangNoCount() -> (bUseSuperBangNoCount:bool))
   > SetActiveWidgetIndex(self=Get ModeSwitcher(), Index=4)
   n4 = Branch(execute=@n3.else, Condition=Get UseSmallVersion() -> (UseSmallVersion:bool))
   n5 = Get InCount() -> (InCount:int=0)
   > SetActiveWidgetIndex(
    self=Get ModeSwitcher()
    Index=Select(Option 0:int=2, Option 1:int=3, Index="integer >= integer"!(A:int=@n5, B:int=1) -> (ReturnValue:bool) {is=title K2Node_CallFunction}) -> (ReturnValue:int)
   )
   n6 = Get InCount() -> (InCount=0)
   SetActiveWidgetIndex(
    execute=@n4.else
    self=Get ModeSwitcher()
    Index=Select(Option 0=1, Option 1=0, Index="Equal (Integer)"(A=@n6, B=0) {is=title K2Node_CallFunction})
   )
  ]}
]
variables [
 bUseSuperBangNoCount {type=boolean}
 Count {type=integer}
 ShouldSuppressBang {type=boolean}
 UseSmallVersion {type=boolean}
]
widgets {
 info {inheritedWidgetCount=1, parentClass=/Script/FortniteUI.FortBangWrapper_NUI, rootWidgetClass=&Overlay, widgetCount=40}
 widgets [
  Overlay_0 {widgetClassPath=&Overlay, widget=Overlay_0}
  TutorialBorder {
   widgetClassPath=&CommonBorder
   bIsVariable=true
   parent=Overlay_0
   widget=TutorialBorder
   properties {
    visibility=Collapsed
    background {
     bIsDynamicallyLoaded=false
     drawAs=Box
     imageType=NoImage
     mirroring=NoMirror
     resourceName=None
     resourceObject=/FortUILibrary/Styles/Border/Materials/M-TutorialTabGlow
     tiling=NoTile
     imageSize=<18 18>
     margin=<ltrb 0.4 0.4 0.4 0.4>
     outlineSettings {
      bUseBrushTransparency=false
      roundingType=HalfHeightRadius
      width=0
      color {colorUseRule=UseColor_Specified, specifiedColor=<rgba 0 0 0 0>}
      cornerRadii=<0 0 0 0>
     }
     tintColor {colorUseRule=UseColor_Specified, specifiedColor=<rgba 1 1 1 1>}
     uVRegion {bIsValid=false, max=<0 0>, min=<0 0>}
    }
    padding=<ltrb 0 0 0 0>
   }
   slot {class=OverlaySlot, properties {horizontalAlignment=HAlign_Fill, verticalAlignment=VAlign_Fill}}
  }
  Content {
   widgetClassPath=&NamedSlot
   bIsVariable=true
   parent=Overlay_0
   widget=Content
   properties {bExposeOnInstanceOnly=true}
   slot {class=OverlaySlot, properties {horizontalAlignment=HAlign_Fill, verticalAlignment=VAlign_Fill}}
  }
  BangBox {
   widgetClassPath=&SizeBox
   bIsVariable=true
   parent=Overlay_0
   widget=BangBox
   slot {class=OverlaySlot, properties {horizontalAlignment=HAlign_Right}}
  }
  ModeSwitcher {widgetClassPath=&WidgetSwitcher, bIsVariable=true, parent=BangBox, widget=ModeSwitcher, slot {class=SizeBoxSlot}}
  BigBang {
   widgetClassPath=&SizeBox
   bIsVariable=true
   parent=ModeSwitcher
   widget=BigBang
   properties {heightOverride=45, minDesiredWidth=25, renderTransform {angle=0, scale=<1 1>, shear=<0 0>, translation=<15 -10>}}
   slot {class=WidgetSwitcherSlot, properties {horizontalAlignment=HAlign_Right, verticalAlignment=VAlign_Top}}
  }
  Overlay_1 {widgetClassPath=&Overlay, parent=BigBang, widget=Overlay_1, slot {class=SizeBoxSlot}}
  CommonBorder_2 {
   widgetClassPath=&CommonBorder
   parent=Overlay_1
   widget=CommonBorder_2
   properties {
    style=/FortUILibrary/Styles/Border/Border-Bang.Border-Bang_C
    visibility=HitTestInvisible
    background {
     bIsDynamicallyLoaded=false
     drawAs=Box
     imageType=NoImage
     mirroring=NoMirror
     resourceName=None
     resourceObject=/FortUILibrary/Styles/Border/Materials/M-Border-Bang
     tiling=NoTile
     imageSize=<45 45>
     margin=<ltrb 0.5962 0 0.6991 0>
     outlineSettings {
      bUseBrushTransparency=false
      roundingType=HalfHeightRadius
      width=0
      color {colorUseRule=UseColor_Specified, specifiedColor=<rgba 0 0 0 0>}
      cornerRadii=<0 0 0 0>
     }
     tintColor {colorUseRule=UseColor_Specified, specifiedColor=<rgba 1 1 1 1>}
     uVRegion {bIsValid=false, max=<0 0>, min=<0 0>}
    }
    brushColor=<rgba 0 0 0 0.6>
    padding=<ltrb 20 0 20 0>
    renderTransform {angle=0, scale=<1 1>, shear=<0 0>, translation=<1 1>}
   }
   slot {class=OverlaySlot, properties {horizontalAlignment=HAlign_Fill, verticalAlignment=VAlign_Fill}}
  }
  Border_NoCount {
   widgetClassPath=&CommonBorder
   parent=Overlay_1
   widget=Border_NoCount
   properties {
    style=&Border-Bang_C
    visibility=HitTestInvisible
    background {
     bIsDynamicallyLoaded=false
     drawAs=Box
     imageType=NoImage
     mirroring=NoMirror
     resourceName=None
     resourceObject=&M-Border-Bang
     tiling=NoTile
     imageSize=<45 45>
     margin=<ltrb 0.5962 0 0.6991 0>
     outlineSettings {
      bUseBrushTransparency=false
      roundingType=HalfHeightRadius
      width=0
      color {colorUseRule=UseColor_Specified, specifiedColor=<rgba 0 0 0 0>}
      cornerRadii=<0 0 0 0>
     }
     tintColor {colorUseRule=UseColor_Specified, specifiedColor=<rgba 1 1 1 1>}
     uVRegion {bIsValid=false, max=<0 0>, min=<0 0>}
    }
    padding=<ltrb 20 0 20 0>
   }
   slot {class=OverlaySlot, properties {horizontalAlignment=HAlign_Fill, verticalAlignment=VAlign_Fill}}
  }
  SizeBox_1 {
   widgetClassPath=&SizeBox
   parent=Overlay_1
   widget=SizeBox_1
   properties {bOverride_HeightOverride=true, bOverride_MinDesiredWidth=true, heightOverride=45, minDesiredWidth=13.6809}
   slot {class=OverlaySlot, properties {horizontalAlignment=HAlign_Fill, verticalAlignment=VAlign_Fill, padding=<ltrb 20 0 20 0>}}
  }
  CommonTextBlock_68 {
   widgetClassPath=&CommonTextBlock
   parent=SizeBox_1
   widget=CommonTextBlock_68
   properties {
    text="!"
    colorAndOpacity {colorUseRule=UseColor_Specified, specifiedColor=<rgba 0 0 0 1>}
    font {
     bForceMonospaced=false
     bMaterialIsStencil=false
     fontMaterial=null
     fontObject=&Burbank
     letterSpacing=0
     monospacedWidth=1
     size=28
     skewAmount=0
     typefaceFontName=Burbank-Black
     outlineSettings {
      bApplyOutlineToDropShadows=false
      bMiteredCorners=false
      bSeparateFillAlpha=false
      outlineMaterial=null
      outlineSize=0
      outlineColor=<rgba 0 0 0 1>
     }
    }
    renderTransform {angle=0, scale=<1 1>, shear=<-8 0>, translation=<0 0>}
    shadowOffset=<0 0>
   }
   slot {class=SizeBoxSlot, properties {horizontalAlignment=HAlign_Center, verticalAlignment=VAlign_Top}}
  }
  BigBangCount {
   widgetClassPath=&SizeBox
   bIsVariable=true
   parent=ModeSwitcher
   widget=BigBangCount
   properties {renderTransform {angle=0, scale=<1 1>, shear=<0 0>, translation=<15 -10>}}
   slot {class=WidgetSwitcherSlot, properties {horizontalAlignment=HAlign_Right, verticalAlignment=VAlign_Top}}
  }
  Overlay_2 {widgetClassPath=&Overlay, parent=BigBangCount, widget=Overlay_2, slot {class=SizeBoxSlot}}
  CommonBorder_3 {
   widgetClassPath=&CommonBorder
   parent=Overlay_2
   widget=CommonBorder_3
   properties {
    style=&Border-Bang_C
    visibility=HitTestInvisible
    background {
     bIsDynamicallyLoaded=false
     drawAs=Box
     imageType=NoImage
     mirroring=NoMirror
     resourceName=None
     resourceObject=&M-Border-Bang
     tiling=NoTile
     imageSize=<45 45>
     margin=<ltrb 0.5962 0 0.6991 0>
     outlineSettings {
      bUseBrushTransparency=false
      roundingType=HalfHeightRadius
      width=0
      color {colorUseRule=UseColor_Specified, specifiedColor=<rgba 0 0 0 0>}
      cornerRadii=<0 0 0 0>
     }
     tintColor {colorUseRule=UseColor_Specified, specifiedColor=<rgba 1 1 1 1>}
     uVRegion {bIsValid=false, max=<0 0>, min=<0 0>}
    }
    brushColor=<rgba 0 0 0 0.6>
    padding=<ltrb 20 0 20 0>
    renderTransform {angle=0, scale=<1 1>, shear=<0 0>, translation=<1 1>}
   }
   slot {class=OverlaySlot, properties {horizontalAlignment=HAlign_Fill, verticalAlignment=VAlign_Fill}}
  }
  Border_Count {
   widgetClassPath=&CommonBorder
   parent=Overlay_2
   widget=Border_Count
   properties {
    style=&Border-Bang_C
    visibility=HitTestInvisible
    background {
     bIsDynamicallyLoaded=false
     drawAs=Box
     imageType=NoImage
     mirroring=NoMirror
     resourceName=None
     resourceObject=&M-Border-Bang
     tiling=NoTile
     imageSize=<45 45>
     margin=<ltrb 0.5962 0 0.6991 0>
     outlineSettings {
      bUseBrushTransparency=false
      roundingType=HalfHeightRadius
      width=0
      color {colorUseRule=UseColor_Specified, specifiedColor=<rgba 0 0 0 0>}
      cornerRadii=<0 0 0 0>
     }
     tintColor {colorUseRule=UseColor_Specified, specifiedColor=<rgba 1 1 1 1>}
     uVRegion {bIsValid=false, max=<0 0>, min=<0 0>}
    }
    padding=<ltrb 20 0 20 0>
   }
   slot {class=OverlaySlot, properties {horizontalAlignment=HAlign_Fill, verticalAlignment=VAlign_Fill}}
  }
  SizeBox_2 {
   widgetClassPath=&SizeBox
   parent=Overlay_2
   widget=SizeBox_2
   properties {bOverride_HeightOverride=true, bOverride_MinDesiredWidth=true, heightOverride=45, minDesiredWidth=25}
   slot {class=OverlaySlot, properties {horizontalAlignment=HAlign_Fill, verticalAlignment=VAlign_Fill, padding=<ltrb 20 0 20 0>}}
  }
  CountText {
   widgetClassPath=&CommonTextBlock
   bIsVariable=true
   parent=SizeBox_2
   widget=CountText
   properties {
    text=25
    colorAndOpacity {colorUseRule=UseColor_Specified, specifiedColor=<rgba 0 0 0 1>}
    font {
     bForceMonospaced=false
     bMaterialIsStencil=false
     fontMaterial=null
     fontObject=&BurbankSmall
     letterSpacing=0
     monospacedWidth=1
     size=24
     skewAmount=0
     typefaceFontName=Black
     outlineSettings {
      bApplyOutlineToDropShadows=false
      bMiteredCorners=false
      bSeparateFillAlpha=false
      outlineMaterial=null
      outlineSize=0
      outlineColor=<rgba 0 0 0 1>
     }
    }
    renderTransform {angle=0, scale=<1 1>, shear=<-8 0>, translation=<0 0>}
    shadowOffset=<0 0>
   }
   slot {class=SizeBoxSlot, properties {verticalAlignment=VAlign_Center}}
  }
  SmallBang {
   widgetClassPath=&SizeBox
   bIsVariable=true
   parent=ModeSwitcher
   widget=SmallBang
   properties {renderTransform {angle=0, scale=<1 1>, shear=<0 0>, translation=<4 -2>}}
   slot {class=WidgetSwitcherSlot, properties {horizontalAlignment=HAlign_Center, verticalAlignment=VAlign_Center}}
  }
  Overlay_3 {widgetClassPath=&Overlay, parent=SmallBang, widget=Overlay_3, slot {class=SizeBoxSlot}}
  CommonBorder_4 {
   widgetClassPath=&CommonBorder
   parent=Overlay_3
   widget=CommonBorder_4
   properties {
    horizontalAlignment=HAlign_Center
    visibility=HitTestInvisible
    background {
     bIsDynamicallyLoaded=false
     drawAs=Box
     imageType=NoImage
     mirroring=NoMirror
     resourceName=None
     resourceObject=&M-Border-Bang
     tiling=NoTile
     imageSize=<32 32>
     margin=<ltrb 0.4 0 0.4 0>
     outlineSettings {
      bUseBrushTransparency=false
      roundingType=HalfHeightRadius
      width=0
      color {colorUseRule=UseColor_Specified, specifiedColor=<rgba 0 0 0 0>}
      cornerRadii=<0 0 0 0>
     }
     tintColor {colorUseRule=UseColor_Specified, specifiedColor=<rgba 1 1 1 1>}
     uVRegion {bIsValid=false, max=<0 0>, min=<0 0>}
    }
    brushColor=<rgba 0 0 0 0.6>
    padding=<ltrb 0 0 0 0>
    renderTransform {angle=0, scale=<1 1>, shear=<0 0>, translation=<1 1>}
   }
   slot {class=OverlaySlot, properties {horizontalAlignment=HAlign_Fill, verticalAlignment=VAlign_Fill}}
  }
  Border_Small {
   widgetClassPath=&CommonBorder
   parent=Overlay_3
   widget=Border_Small
   properties {
    horizontalAlignment=HAlign_Center
    visibility=HitTestInvisible
    background {
     bIsDynamicallyLoaded=false
     drawAs=Box
     imageType=NoImage
     mirroring=NoMirror
     resourceName=None
     resourceObject=&M-Border-Bang
     tiling=NoTile
     imageSize=<32 32>
     margin=<ltrb 0.5 0 0.5 0>
     outlineSettings {
      bUseBrushTransparency=false
      roundingType=HalfHeightRadius
      width=0
      color {colorUseRule=UseColor_Specified, specifiedColor=<rgba 0 0 0 0>}
      cornerRadii=<0 0 0 0>
     }
     tintColor {colorUseRule=UseColor_Specified, specifiedColor=<rgba 1 1 1 1>}
     uVRegion {bIsValid=false, max=<0 0>, min=<0 0>}
    }
    padding=<ltrb 0 0 0 0>
   }
   slot {class=OverlaySlot, properties {horizontalAlignment=HAlign_Fill, verticalAlignment=VAlign_Fill}}
  }
  SizeBox_3 {
   widgetClassPath=&SizeBox
   parent=Overlay_3
   widget=SizeBox_3
   properties {bOverride_HeightOverride=true, bOverride_WidthOverride=true, heightOverride=24, widthOverride=24}
   slot {class=OverlaySlot, properties {horizontalAlignment=HAlign_Center, verticalAlignment=VAlign_Fill}}
  }
  CommonTextBlock {
   widgetClassPath=&CommonTextBlock
   parent=SizeBox_3
   widget=CommonTextBlock
   properties {
    text="!"
    colorAndOpacity {colorUseRule=UseColor_Specified, specifiedColor=<rgba 0 0 0 1>}
    font {
     bForceMonospaced=false
     bMaterialIsStencil=false
     fontMaterial=null
     fontObject=&Burbank
     letterSpacing=0
     monospacedWidth=1
     size=16
     skewAmount=0
     typefaceFontName=Burbank-Black
     outlineSettings {
      bApplyOutlineToDropShadows=false
      bMiteredCorners=false
      bSeparateFillAlpha=false
      outlineMaterial=null
      outlineSize=0
      outlineColor=<rgba 0 0 0 1>
     }
    }
    margin=<ltrb 0 -4 0 -4>
    renderTransform {angle=0, scale=<1 1>, shear=<-8 0>, translation=<0 0>}
    shadowOffset=<0 0>
   }
   slot {class=SizeBoxSlot, properties {horizontalAlignment=HAlign_Center, verticalAlignment=VAlign_Center}}
  }
  SmallBangCount {
   widgetClassPath=&SizeBox
   bIsVariable=true
   parent=ModeSwitcher
   widget=SmallBangCount
   properties {renderTransform {angle=0, scale=<1 1>, shear=<0 0>, translation=<4 -2>}}
   slot {class=WidgetSwitcherSlot, properties {horizontalAlignment=HAlign_Center, verticalAlignment=VAlign_Center}}
  }
  Overlay_4 {widgetClassPath=&Overlay, parent=SmallBangCount, widget=Overlay_4, slot {class=SizeBoxSlot}}
  CommonBorder_0 {
   widgetClassPath=&CommonBorder
   parent=Overlay_4
   widget=CommonBorder_0
   properties {
    horizontalAlignment=HAlign_Center
    visibility=HitTestInvisible
    background {
     bIsDynamicallyLoaded=false
     drawAs=Box
     imageType=NoImage
     mirroring=NoMirror
     resourceName=None
     resourceObject=&M-Border-Bang
     tiling=NoTile
     imageSize=<32 32>
     margin=<ltrb 0.5 0 0.5 0>
     outlineSettings {
      bUseBrushTransparency=false
      roundingType=HalfHeightRadius
      width=0
      color {colorUseRule=UseColor_Specified, specifiedColor=<rgba 0 0 0 0>}
      cornerRadii=<0 0 0 0>
     }
     tintColor {colorUseRule=UseColor_Specified, specifiedColor=<rgba 1 1 1 1>}
     uVRegion {bIsValid=false, max=<0 0>, min=<0 0>}
    }
    brushColor=<rgba 0 0 0 0.6>
    padding=<ltrb 0 0 0 0>
    renderTransform {angle=0, scale=<1 1>, shear=<0 0>, translation=<1 1>}
   }
   slot {class=OverlaySlot, properties {horizontalAlignment=HAlign_Fill, verticalAlignment=VAlign_Fill}}
  }
  CommonBorder_1 {
   widgetClassPath=&CommonBorder
   parent=Overlay_4
   widget=CommonBorder_1
   properties {
    horizontalAlignment=HAlign_Center
    visibility=HitTestInvisible
    background {
     bIsDynamicallyLoaded=false
     drawAs=Box
     imageType=NoImage
     mirroring=NoMirror
     resourceName=None
     resourceObject=&M-Border-Bang
     tiling=NoTile
     imageSize=<32 32>
     margin=<ltrb 0.5 0 0.5 0>
     outlineSettings {
      bUseBrushTransparency=false
      roundingType=HalfHeightRadius
      width=0
      color {colorUseRule=UseColor_Specified, specifiedColor=<rgba 0 0 0 0>}
      cornerRadii=<0 0 0 0>
     }
     tintColor {colorUseRule=UseColor_Specified, specifiedColor=<rgba 1 1 1 1>}
     uVRegion {bIsValid=false, max=<0 0>, min=<0 0>}
    }
    padding=<ltrb 0 0 0 0>
   }
   slot {class=OverlaySlot, properties {horizontalAlignment=HAlign_Fill, verticalAlignment=VAlign_Fill}}
  }
  SizeBox_4 {
   widgetClassPath=&SizeBox
   parent=Overlay_4
   widget=SizeBox_4
   properties {bOverride_HeightOverride=true, bOverride_WidthOverride=true, heightOverride=24, widthOverride=24}
   slot {class=OverlaySlot, properties {horizontalAlignment=HAlign_Center, verticalAlignment=VAlign_Fill, padding=<ltrb 10 0 10 0>}}
  }
  CommonTextBlock_0 {
   widgetClassPath=&CommonTextBlock
   bIsVariable=true
   parent=SizeBox_4
   widget=CommonTextBlock_0
   properties {
    justification=Center
    text=88
    colorAndOpacity {colorUseRule=UseColor_Specified, specifiedColor=<rgba 0 0 0 1>}
    font {
     bForceMonospaced=false
     bMaterialIsStencil=false
     fontMaterial=null
     fontObject=&BurbankSmall
     letterSpacing=0
     monospacedWidth=1
     size=14
     skewAmount=0
     typefaceFontName=Black
     outlineSettings {
      bApplyOutlineToDropShadows=false
      bMiteredCorners=false
      bSeparateFillAlpha=false
      outlineMaterial=null
      outlineSize=0
      outlineColor=<rgba 0 0 0 1>
     }
    }
    renderTransform {angle=0, scale=<1 1>, shear=<-8 0>, translation=<0 0>}
    shadowOffset=<0 0>
   }
   slot {class=SizeBoxSlot, properties {verticalAlignment=VAlign_Center}}
  }
  SuperBang {
   widgetClassPath=&SizeBox
   bIsVariable=true
   parent=ModeSwitcher
   widget=SuperBang
   properties {heightOverride=45, minDesiredWidth=25}
   slot {class=WidgetSwitcherSlot, properties {horizontalAlignment=HAlign_Right, verticalAlignment=VAlign_Top}}
  }
  FlashingBang {
   widgetClassPath=&Overlay
   parent=SuperBang
   widget=FlashingBang
   properties {renderTransform {angle=0, scale=<1 1>, shear=<0 0>, translation=<8 -4>}}
   slot {class=SizeBoxSlot}
  }
  Image_2 {
   widgetClassPath=&Image
   parent=FlashingBang
   widget=Image_2
   properties {
    visibility=SelfHitTestInvisible
    brush {
     bIsDynamicallyLoaded=false
     drawAs=Image
     imageType=NoImage
     mirroring=NoMirror
     resourceName=None
     resourceObject=/FortUILibrary/Styles/Border/Materials/M_FlashingBang_Off
     tiling=NoTile
     imageSize=<32 32>
     margin=<ltrb 0 0 0 0>
     outlineSettings {
      bUseBrushTransparency=false
      roundingType=HalfHeightRadius
      width=0
      color {colorUseRule=UseColor_Specified, specifiedColor=<rgba 0 0 0 0>}
      cornerRadii=<0 0 0 0>
     }
     tintColor {colorUseRule=UseColor_Specified, specifiedColor=<rgba 1 1 1 1>}
     uVRegion {bIsValid=false, max=<0 0>, min=<0 0>}
    }
   }
   slot {class=OverlaySlot}
  }
  CommonTextBlock_1 {
   widgetClassPath=&CommonTextBlock
   parent=FlashingBang
   widget=CommonTextBlock_1
   properties {
    style=/FortUILibrary/Styles/Typography/Foundation/TextStyle-Header-M-Yellow.TextStyle-Header-M-Yellow_C
    text="!"
    colorAndOpacity {colorUseRule=UseColor_Specified, specifiedColor=<rgba 0.9473 0.9911 0.0666 1>}
    font {
     bForceMonospaced=false
     bMaterialIsStencil=false
     fontMaterial=null
     fontObject=&Burbank
     letterSpacing=0
     monospacedWidth=1
     size=32
     skewAmount=0
     typefaceFontName=Burbank-Black
     outlineSettings {
      bApplyOutlineToDropShadows=false
      bMiteredCorners=false
      bSeparateFillAlpha=false
      outlineMaterial=null
      outlineSize=0
      outlineColor=<rgba 0 0 0 1>
     }
    }
    renderTransform {angle=5, scale=<0.75 0.625>, shear=<-8 0>, translation=<0 0>}
    shadowOffset=<0 0>
   }
   slot {class=OverlaySlot, properties {horizontalAlignment=HAlign_Center, verticalAlignment=VAlign_Center, padding=<ltrb 0 -12 0 -12>}}
  }
  LockBang {
   widgetClassPath=&SizeBox
   bIsVariable=true
   parent=ModeSwitcher
   widget=LockBang
   properties {renderTransform {angle=0, scale=<1 1>, shear=<0 0>, translation=<4 -2>}}
   slot {class=WidgetSwitcherSlot, properties {horizontalAlignment=HAlign_Center, verticalAlignment=VAlign_Center}}
  }
  Overlay {widgetClassPath=&Overlay, parent=LockBang, widget=Overlay, slot {class=SizeBoxSlot}}
  CommonBorder {
   widgetClassPath=&CommonBorder
   parent=Overlay
   widget=CommonBorder
   properties {
    horizontalAlignment=HAlign_Center
    visibility=HitTestInvisible
    background {
     bIsDynamicallyLoaded=false
     drawAs=Box
     imageType=NoImage
     mirroring=NoMirror
     resourceName=None
     resourceObject=&M-Border-Bang
     tiling=NoTile
     imageSize=<32 32>
     margin=<ltrb 0.4 0 0.4 0>
     outlineSettings {
      bUseBrushTransparency=false
      roundingType=HalfHeightRadius
      width=0
      color {colorUseRule=UseColor_Specified, specifiedColor=<rgba 0 0 0 0>}
      cornerRadii=<0 0 0 0>
     }
     tintColor {colorUseRule=UseColor_Specified, specifiedColor=<rgba 1 1 1 1>}
     uVRegion {bIsValid=false, max=<0 0>, min=<0 0>}
    }
    brushColor=<rgba 0 0 0 0.6>
    padding=<ltrb 0 0 0 0>
    renderTransform {angle=0, scale=<1 1>, shear=<0 0>, translation=<1 1>}
   }
   slot {class=OverlaySlot, properties {horizontalAlignment=HAlign_Fill, verticalAlignment=VAlign_Fill}}
  }
  Border_Small_1 {
   widgetClassPath=&CommonBorder
   parent=Overlay
   widget=Border_Small_1
   properties {
    horizontalAlignment=HAlign_Center
    visibility=HitTestInvisible
    background {
     bIsDynamicallyLoaded=false
     drawAs=Box
     imageType=NoImage
     mirroring=NoMirror
     resourceName=None
     resourceObject=&M-Border-Bang
     tiling=NoTile
     imageSize=<32 32>
     margin=<ltrb 0.5 0 0.5 0>
     outlineSettings {
      bUseBrushTransparency=false
      roundingType=HalfHeightRadius
      width=0
      color {colorUseRule=UseColor_Specified, specifiedColor=<rgba 0 0 0 0>}
      cornerRadii=<0 0 0 0>
     }
     tintColor {colorUseRule=UseColor_Specified, specifiedColor=<rgba 1 1 1 1>}
     uVRegion {bIsValid=false, max=<0 0>, min=<0 0>}
    }
    padding=<ltrb 0 0 0 0>
   }
   slot {class=OverlaySlot, properties {horizontalAlignment=HAlign_Fill, verticalAlignment=VAlign_Fill}}
  }
  SizeBox {
   widgetClassPath=&SizeBox
   parent=Overlay
   widget=SizeBox
   properties {bOverride_HeightOverride=true, bOverride_WidthOverride=true, heightOverride=30, widthOverride=30}
   slot {class=OverlaySlot, properties {horizontalAlignment=HAlign_Center, verticalAlignment=VAlign_Fill}}
  }
  LockImage {
   widgetClassPath=&FortLazyImage
   bIsVariable=true
   parent=SizeBox
   widget=LockImage
   properties {brush {
     bIsDynamicallyLoaded=false
     drawAs=Image
     imageType=FullColor
     mirroring=NoMirror
     resourceName=None
     resourceObject=/Game/UI/Foundation/Textures/Icons/Navigation/AV_Padlock_32
     tiling=NoTile
     imageSize=<32 32>
     margin=<ltrb 0 0 0 0>
     outlineSettings {
      bUseBrushTransparency=false
      roundingType=HalfHeightRadius
      width=0
      color {colorUseRule=UseColor_Specified, specifiedColor=<rgba 0 0 0 0>}
      cornerRadii=<0 0 0 0>
     }
     tintColor {colorUseRule=UseColor_Specified, specifiedColor=<rgba 0 0 0 1>}
     uVRegion {bIsValid=false, max=<0 0>, min=<0 0>}
    }}
   slot {class=SizeBoxSlot, properties {horizontalAlignment=HAlign_Center, verticalAlignment=VAlign_Center, padding=<ltrb 5 5 6 5>}}
  }
  ContentSlot {widgetClassPath=&NamedSlot, bInherited=true, widget=null}
 ]
}
